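# Translucent, unlit, two-sided material for Reality Tendril environment effects, with vertex colors driving hue shift and top/river variation.
# Color blends two tints, animates hue through vertex interpolators, and scales with time of day and camera distance.
# Opacity multiplies panned wave textures, a line mask and UV edge gradients, adding depth fade and Fresnel by shading path and quality level.
# A distance-synced Z offset follows the camera's relative position.
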Material /RealitySeedGameplay/Environment/Assets_RealityTendril/M_Reality_Tendril
graph {
 properties {blendMode=BLEND_Translucent, pixelDepthOffsetMode=PDOM_Legacy, shadingModel=MSM_Unlit, twoSided=true, usedShadingModels=Unlit}
 nodes [
  n1 = MaterialFunctionCall(
   Fade_Start_Distance=ScalarParameter() {defaultValue=1500, parameterName=Start_Z_Lift}
   Fade_End_Distance=ScalarParameter() {parameterName=End_Z_Lift}
  ) {materialFunction=&MF_Distance_SyncedFade_FX}
  n2 = LinearInterpolate(B=ScalarParameter() {defaultValue=1, group=WPO, parameterName=ZOffset}, Alpha=@n1.Result)
  n3 = CameraPositionWS()
  n4 = ObjectPositionWS()
  n5 = Clamp(None=ComponentMask(None=Subtract(A=@n3, B=@n4)) {b=true}) {minDefault=-1}
  Add(B=MaterialFunctionCall(Z=Multiply(A=@n2, B=@n5)) {materialFunction=&MakeFloat3}.Result)
  Constant() {r=4}
  MaterialFunctionCall() {materialFunction=&Near_Camera_Fade}
  n6 = LinearInterpolate(B=ScalarParameter() {defaultValue=2, parameterName=DistanceColorMult}, Alpha=@n1.Result) {constA=1}
  n7 = MaterialFunctionCall(
   Value_Morning=ScalarParameter() {defaultValue=1, parameterName="Value_Morning (S)"}
   Value_Day=ScalarParameter() {defaultValue=1, parameterName="Value_Day (S)"}
   Value_Evening=ScalarParameter() {defaultValue=1, parameterName="Value_Evening (S)"}
   Value_Night=ScalarParameter() {defaultValue=1, parameterName="Value_Night (S)"}
  ) {materialFunction=&Scalar_Time_of_Day}
  n8 = ScalarParameter() {defaultValue=1, parameterName=MultColor_Top}
  n9 = ScalarParameter() {defaultValue=1, parameterName=MultColor_River}
  n10 = VertexColor()
  n11 = VertexColor()
  n12 = ComponentMask(None=TextureCoordinate() {coordinateIndex=1}) {g=true}
  n13 = Time() {period=2}
  n14 = Multiply(A=ScalarParameter() {defaultValue=1, parameterName=HueShiftSpeed}, B=LinearInterpolate(Alpha=VertexColor().A) {constA=0.6})
  n15 = Sine(None=Frac(None=Multiply(A=@n13, B=@n14)))
  n16 = Multiply(A=Multiply(A=Add(A=@n12, B=@n15), B=@n15), B=ScalarParameter() {defaultValue=1, parameterName=HueShiftMult})
  n17 = VectorParameter() {parameterName=BaseColor, defaultValue=<rgba 5 2 1 1>}
  n18 = VectorParameter() {parameterName=BaseColorHigh, defaultValue=<rgba 5 2 1 1>}
  n19 = VertexColor()
  n20 = VertexInterpolator(VS=MaterialFunctionCall(Hue Shift Percentage=Multiply(A=@n11.R, B=@n16), Texture=LinearInterpolate(A=@n17.RGB, B=@n18.RGB, Alpha=@n19.A)) {materialFunction=&HueShift}.Result)
  n21 = ScalarParameter() {parameterName=Desaturation}
  Multiply(
   A=VertexInterpolator(VS=Multiply(A=@n6, B=@n7.Result)).PS
   B=Multiply(A=LinearInterpolate(A=@n8, B=@n9, Alpha=@n10.B), B=Desaturation(None=@n20.PS, Fraction=@n21))
  )
  Multiply(A=@n19.A)
  n22 = ScalarParameter() {parameterName=XTileOffset}
  n23 = Multiply(
   A=TextureCoordinate() {uTiling=0.2, vTiling=4}
   B=ComponentMask(None=VectorParameter() {parameterName=TileMult, defaultValue=<rgba 1 1 0 1>}.RGB) {g=true, r=true}
  )
  n24 = ComponentMask(None=@n23) {r=true}
  n25 = ComponentMask(None=@n23) {g=true}
  n26 = ComponentMask(None=TextureCoordinate() {coordinateIndex=1}) {g=true}
  n27 = Multiply(A=ComponentMask(None=VectorParameter() {parameterName=TileMult, defaultValue=<rgba 1 1 0 1>}.RGB) {g=true}, B=@n26)
  n28 = Panner(Coordinate=AppendVector(A=Add(A=@n22, B=@n24), B=Add(A=@n25, B=@n27)), Speed=Constant2Vector() {g=-0.1})
  n29 = Panner(Coordinate=TextureCoordinate() {coordinateIndex=1}) {speedY=0.1}
  n30 = ComponentMask(None=VectorParameter() {parameterName=TileMultMask, defaultValue=<rgba 1 1 0 1>}.RGB) {g=true, r=true}
  n31 = Multiply(
   A=LinearInterpolate(
    A=TextureSampleParameter2D(UVs=@n28) {parameterName=MainTexture_01, texture=/Game/Athena/Environments/Nexus/Textures/T_ZeroPoint_MagneticWaves}.G
    B=TextureSampleParameter2D(UVs=@n28) {parameterName=MainTexture_02, texture=&T_ZeroPoint_MagneticWaves}.G
    Alpha=@n26
   )
   B=TextureSampleParameter2D(UVs=Multiply(A=@n29, B=@n30)) {parameterName=MaskTexture, texture=/RealitySeedGameplay/Environment/Assets_Daisy/FX/Textures/T_Lines_Base_01}.G
  )
  n32 = MaterialFunctionCall(UV Channel=TextureCoordinate()) {materialFunction=&LinearGradient}
  n33 = OneMinus(None=@n32.UGradient)
  n34 = ScalarParameter() {defaultValue=1, parameterName=Mult_Mask_Top}
  n35 = ScalarParameter() {defaultValue=1, parameterName=Mult_Mask_River}
  n36 = VertexColor()
  n37 = Power(Base=Multiply(A=Multiply(A=@n32.UGradient, B=@n33), B=LinearInterpolate(A=@n34, B=@n35, Alpha=@n36.B))) {constExponent=4}
  n38 = MaterialFunctionCall(UV Channel=TextureCoordinate()) {materialFunction=&LinearGradient}
  n39 = Multiply(A=Multiply(A=@n38.VGradient, B=OneMinus(None=@n38.VGradient)), B=Constant() {r=4})
  n40 = Multiply(
   A=ScalarParameter() {defaultValue=1, parameterName=OpacityMult}
   B=Multiply(A=Multiply(A=@n31, B=@n37), B=Power(Base=@n39) {constExponent=4})
  )
  n41 = Multiply(
   A=@n40
   B=MaterialFunctionCall(
    Fade Length=ScalarParameter() {defaultValue=512, parameterName=NearFadeDistance}
    Fade Offset=ScalarParameter() {defaultValue=256, parameterName=NearFadeDistance_Start_Offset}
   ) {materialFunction=&CameraDepthFade}.Result
  )
  n42 = ShadingPathSwitch(Default=@n41, Forward=@n40, Mobile=@n40)
  n43 = Fresnel(ExponentIn=ScalarParameter() {defaultValue=0.5, parameterName=FresnelFade_Exponent}) {baseReflectFraction=0.01, exponent=0.5}
  QualitySwitch(Default=@n42, Low=@n40, High=@n42, Medium=@n40, Epic=Multiply(A=OneMinus(None=@n43), B=@n41))
  VertexColor()
 ]
}
